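# Unlit masked Niagara material for the IndieBucket skydiving contrail ribbon.
# Emissive blends dark and light colors, each switched by time of day, using masks from named reroutes.
# A ribbon texture supplies channel masks, while two panning noise layers, driven by dynamic parameters and per-particle random offsets, form dashed lines.
# Panning cell-texture erosion along U is thresholded to a hard cutout, with dithered fading near the camera.
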Material /BRCosmetics/Athena/Effects/Particles/Skydiving/NS_Contrail_IndieBucket/M_Contrail_IndieBucket_Ribbon
graph {
 properties {
  bAutomaticallySetUsageInEditor=false
  bCastRayTracedShadows=false
  bUsedWithNiagaraSprites=true
  blendMode=BLEND_Masked
  pixelDepthOffsetMode=PDOM_Legacy
  shadingModel=MSM_Unlit
  usedShadingModels=Unlit
 }
 nodes [
  n1 = VectorParameter() {parameterName=DarkColor, defaultValue=<rgba 0.1413 0.0307 0.0578 1>}
  n2 = VectorParameter() {parameterName=DarkColor_Night, defaultValue=<rgba 0.01 0.03 0.05 1>}
  n3 = VectorParameter() {parameterName=LightColor, defaultValue=<rgba 0.7084 0.624 0.5583 1>}
  n4 = VectorParameter() {parameterName=LightColor_Night, defaultValue=<rgba 0.3 0.45 0.5 1>}
  n5 = Saturate(None=Multiply(A=NamedRerouteUsage()) {constB=10})
  n6 = Multiply(A=NamedRerouteUsage()) {constB=10}
  n7 = OneMinus(None=Multiply(A=OneMinus(None=@n5), B=Saturate(None=@n6)))
  n8 = NamedRerouteUsage()
  MaterialFunctionCall(Emissive=LinearInterpolate(
    A=MaterialFunctionCall(Color_Morning=@n1.RGB, Color_Day=@n1.RGB, Color_Evening=@n1.RGB, Color_Night=@n2.RGB) {materialFunction=&Color_Time_of_Day}.Result
    B=MaterialFunctionCall(Color_Morning=@n3.RGB, Color_Day=@n3.RGB, Color_Evening=@n3.RGB, Color_Night=@n4.RGB) {materialFunction=&Color_Time_of_Day}.Result
    Alpha=Multiply(A=@n7, B=@n8)
   )) {materialFunction=&MF_FX_EmissiveAdjustments}
  n9 = WorldPosition()
  n10 = CameraPositionWS()
  n11 = ScalarParameter() {defaultValue=75, parameterName=FadeMask}
  n12 = Constant()
  n13 = MaterialFunctionCall(Alpha Threshold=OneMinus(None=SphereMask(A=@n9.XYZ, B=@n10, Radius=@n11, Hardness=@n12))) {materialFunction=&DitherTemporalAA_Fortnite}
  n14 = ScalarParameter() {defaultValue=3, parameterName=Dither Power}
  Multiply(B=Multiply(A=Power(Base=@n13.Result, Exp=@n14) {constExponent=3}, B=ScalarParameter() {defaultValue=3, parameterName=Dither Mult}))
  n15 = MaterialFunctionCall(Float2=TextureCoordinate()) {materialFunction=&BreakOutFloat2Components}
  n16 = ScalarParameter() {defaultValue=3, parameterName=ErosionMultRamp_Mult}
  n17 = Saturate(None=Power(Base=Multiply(A=@n15.R, B=@n16) {constB=2}, Exp=ScalarParameter() {defaultValue=5, parameterName=ErosionMultRamp_Power}))
  n18 = MaterialFunctionCall(Float2=TextureCoordinate()) {materialFunction=&BreakOutFloat2Components}
  n19 = Multiply(A=@n18.R) {constB=12}
  n20 = ScalarParameter() {defaultValue=4, parameterName=PanSpeed}
  n21 = Constant() {r=0.1}
  n22 = TextureSample(UVs=Panner(Coordinate=AppendVector(A=@n19, B=@n18.G), Speed=AppendVector(A=@n20, B=@n21))) {
   samplerType=SAMPLERTYPE_LinearColor
   texture=&/T_FX_IndieBucket_Contrail_Cells
  }
  n23 = OneMinus(None=MaterialFunctionCall(Float2=TextureCoordinate()) {materialFunction=&BreakOutFloat2Components}.R)
  n24 = ScalarParameter() {defaultValue=4, parameterName=ErosionPowerRamp_Power}
  n25 = Power(
   Base=@n22.RGB
   Exp=Multiply(A=Power(Base=@n23, Exp=@n24) {constExponent=5}, B=ScalarParameter() {defaultValue=100, parameterName=ErosionPowerRamp_Mult}) {constB=20}
  )
  Multiply(A=If(A=Multiply(A=@n17, B=@n25), "A > B"=Constant() {r=1}, "A < B"=Constant()) {constB=0.05}, B=NamedRerouteUsage())
  n26 = MaterialFunctionCall(Float2=TextureCoordinate()) {materialFunction=&BreakOutFloat2Components}
  n27 = Multiply(
   A=Multiply(A=@n26.R, B=ScalarParameter() {defaultValue=1, group=Dashes, parameterName=Digital Mask Scale})
   B=DynamicParameter() {paramNames [NoiseOffset, NoiseOffset2, DiveTilingMult, DiveTilingMult2]}.DiveTilingMult
  )
  n28 = Add(
   A=Multiply(A=@n26.G, B=ScalarParameter() {defaultValue=0.95, parameterName=YScale})
   B=DynamicParameter() {paramNames [NoiseOffset, NoiseOffset2, DiveTilingMult, DiveTilingMult2]}.NoiseOffset
  )
  n29 = Time()
  n30 = ScalarParameter() {defaultValue=3, group=Dashes, parameterName=Dash Particle Rand Value}
  n31 = LinearInterpolate(B=@n30, Alpha=ParticleRandom()) {constA=1, constB=3}
  n32 = ScalarParameter() {defaultValue=1, group=Dashes, parameterName=DigitalPanSpeed}
  n33 = Multiply(A=ScalarParameter() {defaultValue=0.1, parameterName=XPan}) {constB=0.5}
  n34 = TextureSample(UVs=Panner(Coordinate=AppendVector(A=@n27, B=@n28), Time=Add(A=@n29, B=@n31), Speed=AppendVector(A=@n32, B=@n33)) {speedX=0.1}) {texture=&/T_FX_IndieBucket_Noise_Resize}
  n35 = MaterialFunctionCall(Float2=TextureCoordinate()) {materialFunction=&BreakOutFloat2Components}
  n36 = Multiply(
   A=Multiply(A=@n35.R, B=ScalarParameter() {defaultValue=0.5, group=Dashes, parameterName=Digital Mask Scale Lg})
   B=DynamicParameter() {paramNames [NoiseOffset, NoiseOffset2, DiveTilingMult, DiveTilingMult2]}.DiveTilingMult2
  )
  n37 = Add(
   A=Multiply(A=@n35.G, B=ScalarParameter() {defaultValue=0.95, parameterName=YScaleLG}) {constB=0.95}
   B=DynamicParameter() {paramNames [NoiseOffset, NoiseOffset2, DiveTilingMult, DiveTilingMult2]}.NoiseOffset2
  )
  n38 = Time()
  n39 = LinearInterpolate(B=@n30, Alpha=ParticleRandom()) {constA=1, constB=3}
  n40 = ScalarParameter() {defaultValue=0.5, group=Dashes, parameterName=DigitalPanSpeed Lg}
  n41 = ScalarParameter() {defaultValue=0.1, parameterName=XPan}
  n42 = TextureSample(UVs=Panner(Coordinate=AppendVector(A=@n36, B=@n37), Time=Add(A=@n38, B=@n39), Speed=AppendVector(A=@n40, B=@n41)) {speedX=0.1}) {texture=&T_FX_IndieBucket_Noise_Resize}
  n43 = Saturate(None=Multiply(A=Multiply(A=@n34.R, B=@n42.R) {constA=1}) {constB=10})
  n44 = NamedRerouteUsage()
  n45 = NamedRerouteUsage()
  NamedRerouteDeclaration(None=Add(A=Multiply(A=@n43, B=@n44), B=OneMinus(None=@n45))) {name=Lined, nodeColor=<rgba 1 0.7266 0 1>}
  n46 = TextureSample(UVs=TextureCoordinate()) {texture=&/T_FX_IndieBucket_Ribbon}
  NamedRerouteDeclaration(None=Saturate(None=Multiply(A=@n46.G) {constB=10})) {name=green, nodeColor=<rgba 0 1 0.2734 1>}
  NamedRerouteDeclaration(None=@n46.R) {name=red, nodeColor=<rgba 1 0 0.0937 1>}
  NamedRerouteDeclaration(None=@n46.B) {name=blue, nodeColor=<rgba 0.0312 0 1 1>}
  n47 = MaterialFunctionCall(Float2=TextureCoordinate()) {materialFunction=&BreakOutFloat2Components}
  n48 = SmoothStep(Value=Multiply(A=@n47.R, B=OneMinus(None=@n47.R))) {constMax=0.5}
  Saturate(None=Multiply(A=Power(Base=@n48)) {constB=10})
 ]
}
