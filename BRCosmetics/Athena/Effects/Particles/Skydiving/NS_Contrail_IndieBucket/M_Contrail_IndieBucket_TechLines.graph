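# Masked, unlit, two-sided Niagara material for the tech lines in the IndieBucket skydiving contrail, rendering dot-patterned, spline-thickened strips.
# Dot texture tiles along V by particle tiling scale and spline width, its channels blending inside, outline and shadow colors that shift with time of day.
# Per-particle randoms offset and select color-mask samples for variation.
# Near-camera fades drive far-color tinting and opacity, with a sine mask and dithered ramp threshold shaping visibility.
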Material /BRCosmetics/Athena/Effects/Particles/Skydiving/NS_Contrail_IndieBucket/M_Contrail_IndieBucket_TechLines
graph {
 properties {
  bCastRayTracedShadows=false
  bUsedWithNiagaraMeshParticles=true
  bUsedWithNiagaraSprites=true
  blendMode=BLEND_Masked
  pixelDepthOffsetMode=PDOM_Legacy
  shadingModel=MSM_Unlit
  twoSided=true
  usedShadingModels=Unlit
 }
 nodes [
  Fmod()
  If()
  n1 = MaterialFunctionCall(Alpha Threshold=DynamicParameter() {parameterIndex=3, paramNames [Ramp, "", "", ""]}.Ramp) {materialFunction=&DitherTemporalAA_Fortnite}
  n2 = Saturate(None=Multiply(A=OneMinus(None=@n1.Result)) {constB=3})
  LinearInterpolate(Alpha=@n2) {constB=500}
  n3 = DynamicParameter() {parameterIndex=1, defaultValue=<rgba 1 10 1 1>, paramNames [Random_02, SplineWidth, SplineAnim_Power, SplineAnim_Mult]}
  MaterialFunctionCall(WidthBase=@n3.SplineWidth, WidthTip=@n3.SplineWidth) {materialFunction=&SplineThicken}
  n4 = MaterialFunctionCall(Float2=TextureCoordinate()) {materialFunction=&BreakOutFloat2Components}
  n5 = DynamicParameter() {parameterIndex=1, defaultValue=<rgba 1 10 1 1>, paramNames [Random_02, SplineWidth, SplineAnim_Power, SplineAnim_Mult]}
  n6 = Multiply(A=DynamicParameter() {paramNames [OpacityMult,PanTime,VTilingScale,Random_01]}.VTilingScale, B=Divide(B=@n5.SplineWidth) {constA=100}) {constB=5}
  n7 = TextureSample(UVs=MaterialFunctionCall(X=@n4.R, Y=Multiply(A=@n4.G, B=@n6) {constB=5, desc=UV Tiling Scale}) {materialFunction=&MakeFloat2}.Result) {desc=80perc, texture=&/T_FX_IndieBucket_Dot}
  n8 = NamedRerouteUsage()
  n9 = NamedRerouteUsage()
  n10 = NamedRerouteUsage()
  n11 = NamedRerouteUsage()
  n12 = LinearInterpolate(
   A=LinearInterpolate(A=NamedRerouteUsage(), B=NamedRerouteUsage(), Alpha=@n7.R)
   B=LinearInterpolate(A=Multiply(A=@n8, B=@n9) {constB=0.3}, B=Multiply(A=@n10, B=@n11) {constB=0.3}, Alpha=@n7.R)
   Alpha=@n7.G
  )
  n13 = VectorParameter() {parameterName=FarColor, defaultValue=<rgba 0.6 0.4 0.35 1>}
  n14 = MaterialFunctionCall(
   Color_Morning=@n13.RGB
   Color_Day=@n13.RGB
   Color_Evening=@n13.RGB
   Color_Night=VectorParameter() {parameterName=FarColor_Night, defaultValue=<rgba 0.35 0.4 0.6 1>}.RGB
  ) {materialFunction=&Color_Time_of_Day}
  n15 = ScalarParameter() {defaultValue=500}
  n16 = Multiply(A=Power(Base=MaterialFunctionCall(Distance=@n15) {materialFunction=&Near_Camera_Fade}.Result) {constExponent=3}) {constB=3}
  MaterialFunctionCall(Emissive=LinearInterpolate(A=@n12, B=Multiply(A=@n12, B=@n14.Result) {constB=0.2}, Alpha=Saturate(None=@n16))) {materialFunction=&MF_FX_EmissiveAdjustments}
  n17 = Divide(A=TextureCoordinate())
  n18 = DynamicParameter() {parameterIndex=1, defaultValue=<rgba 1 10 1 1>, paramNames [Random_02, SplineWidth, SplineAnim_Power, SplineAnim_Mult]}
  n19 = Add(A=ComponentMask(None=Sine(None=@n17)) {g=true}, B=@n18.SplineAnim_Mult) {constB=-1}
  n20 = Constant() {r=400}
  Multiply(
   A=Multiply(
    A=Multiply(A=@n7.A, B=@n19)
    B=MaterialFunctionCall(Distance=@n20) {materialFunction=&Near_Camera_Fade}.Result
   )
   B=@n2
  )
  n21 = VectorParameter() {parameterName=LightColor02_Day, defaultValue=<rgba 0.6724 0.3419 0.107 1>}
  n22 = VectorParameter() {parameterName=LightColor02_Night, defaultValue=<rgba 0.0423 0.0723 0.0908 1>}
  n23 = VectorParameter() {parameterName=LightColor_Day, defaultValue=<rgba 0.7084 0.624 0.5583 1>}
  n24 = VectorParameter() {parameterName=LightColor_Night, defaultValue=<rgba 0.3 0.45 0.5 1>}
  NamedRerouteDeclaration(None=LinearInterpolate(
    A=MaterialFunctionCall(Color_Morning=@n21.RGB, Color_Day=@n21.RGB, Color_Evening=@n21.RGB, Color_Night=@n22.RGB) {materialFunction=&Color_Time_of_Day}.Result
    B=MaterialFunctionCall(Color_Morning=@n23.RGB, Color_Day=@n23.RGB, Color_Evening=@n23.RGB, Color_Night=@n24.RGB) {materialFunction=&Color_Time_of_Day}.Result
    Alpha=NamedRerouteUsage()
   )) {name=InsideColor, nodeColor=<rgba 1 0 0.9141 1>}
  n25 = VectorParameter() {parameterName=OutlineColor_Day, defaultValue=<rgba 0.1413 0.0307 0.0578 1>}
  NamedRerouteDeclaration(None=MaterialFunctionCall(
    Color_Morning=@n25.RGB
    Color_Day=@n25.RGB
    Color_Evening=@n25.RGB
    Color_Night=VectorParameter() {parameterName=OutlineColor_Night, defaultValue=<rgba 0.01 0.03 0.05 1>}.RGB
   ) {materialFunction=&Color_Time_of_Day}.Result) {name=OutsideColor, nodeColor=<rgba 1 0.8203 0 1>}
  n26 = VectorParameter() {parameterName=Shadow_Day, defaultValue=<rgba 0.41 0.3681 0.62 1>}
  NamedRerouteDeclaration(None=MaterialFunctionCall(
    Color_Morning=@n26.RGB
    Color_Day=@n26.RGB
    Color_Evening=@n26.RGB
    Color_Night=VectorParameter() {parameterName=Shadow_Night, defaultValue=<rgba 0.2775 0.3216 0.8073 1>}.RGB
   ) {materialFunction=&Color_Time_of_Day}.Result) {name=ShadowColor, nodeColor=<rgba 1 0.0234 0 1>}
  n27 = MaterialFunctionCall(Float2=TextureCoordinate()) {materialFunction=&BreakOutFloat2Components}
  n28 = Multiply(A=DynamicParameter() {paramNames [OpacityMult,PanTime,VTilingScale,Random_01]}.Random_01) {constB=4}
  n29 = Add(A=@n27.G, B=Divide(A=Round(None=@n28)) {constB=5})
  n30 = Divide(A=NamedRerouteUsage()) {constB=5}
  n31 = TextureSample(UVs=AppendVector(A=@n27.R, B=Multiply(A=@n29, B=@n30))) {texture=&/T_FX_IndieBucket_ColorMask_02}
  n32 = MaterialFunctionCall(Float2=TextureCoordinate()) {materialFunction=&BreakOutFloat2Components}
  n33 = DynamicParameter() {parameterIndex=1, defaultValue=<rgba 1 10 1 1>, paramNames [Random_02, SplineWidth, SplineAnim_Power, SplineAnim_Mult]}
  n34 = Multiply(A=@n33.Random_02) {constB=4}
  n35 = Add(A=@n32.G, B=Divide(A=Round(None=@n34)) {constB=5})
  n36 = Divide(A=NamedRerouteUsage()) {constB=5}
  n37 = TextureSample(UVs=AppendVector(A=@n32.R, B=Multiply(A=@n35, B=@n36))) {texture=&T_FX_IndieBucket_ColorMask_02}
  NamedRerouteDeclaration(None=MaterialFunctionCall(
    1=@n31.RGB
    2=Multiply(A=@n31.RGB, B=@n37.RGB)
    3=Constant3Vector() {constant=<rgba 1 1 1 0>}
    A=DynamicParameter() {parameterIndex=2, paramNames [RandomSelector, "", "", ""]}.RandomSelector
   ) {materialFunction=&Lerp_Multiple_Float3}.Lerp 3 Inputs) {name=RandomColorSelect, nodeColor=<rgba 1 0.9375 0 1>}
  NamedRerouteDeclaration(None=@n6) {name=Tiling, nodeColor=<rgba 0 0.8594 1 1>}
 ]
}
