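# Skydiving contrail for the IndieBucket cosmetic: upward-moving mesh dashes in a box, plus two sprite stripes on skeletal mesh sockets.
# Three GPU DottedTechLines emitters are gated by User.Variant, which also sets renderer visibility tags; the original is disabled.
# System timeline drives spawn rate (lower in front-end preview), lifetime and box offset, while glide lean boosts speed.
# Distance from the owner darkens dashes and feeds their material.
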NiagaraSystem /BRCosmetics/Athena/Effects/Particles/Skydiving/NS_Contrail_IndieBucket/NS_Contrail_IndieBucket
dependencies {
 usedDataInterfaces [&NiagaraDataInterfaceCurve, &NiagaraDataInterfaceArrayFloat3, &NiagaraDataInterfaceArrayColor]
 usedDynamicInputs [&SelectLinearColorFromArray]
 usedModules [
  &self:SystemSpawnScript.SystemScriptSource.SystemScriptGraph.NiagaraNodeAssignment_3.SetVariables_#1
  &self:SystemSpawnScript.SystemScriptSource.SystemScriptGraph.NiagaraNodeAssignment_4.SetVariables_#2
  &self:SystemSpawnScript.SystemScriptSource.SystemScriptGraph.NiagaraNodeAssignment_6.SetVariables_#3
  &self:SystemSpawnScript.SystemScriptSource.SystemScriptGraph.NiagaraNodeAssignment_7.SetVariables_#4
  &self:SystemSpawnScript.SystemScriptSource.SystemScriptGraph.NiagaraNodeAssignment_5.SetVariables_#5
  &self:SystemSpawnScript.SystemScriptSource.SystemScriptGraph.NiagaraNodeAssignment_8.SetVariables_#6
  &SystemState
  &Timeline
 ]
}
emitters {
 DottedTechLines {
  bEnabled=false
  simTarget=GPUComputeSim
  emitterUpdateScript [
   &EmitterState()
   &SpawnRate(SpawnRate=&Lerp_Float(A=150, B=650, Alpha=$System.Timeline.NormalizedTime), Use Spawn Probability=false)
   &SpawnRate(SpawnGroup=1, SpawnRate=10, Use Spawn Probability=false) {enabled=false, module=SpawnRate001}
  ]
  particleSpawnScript [
   /SetVariables(Particles.PanSpeed=0.1) {isSetParameters=true, module=SetVariables_#7}
   &InitializeParticle(
    Color Mode=::NewEnumerator0
    Lifetime=&Multiply_Float(
     A=&Add_Float(
      A=&Lerp_Float(
       A=0.5
       B=1
       Alpha=&FloatFromCurve(
        FloatCurve=DataInterface() {propertyValues {Curve {keys [<at 0 0 RCIM_Cubic>, {arriveTangent=1, interpMode=RCIM_Cubic, leaveTangent=1, time=0.7, value=1}, <at 1 1 RCIM_Cubic>]}}}
        CurveIndex=&RandomRangeFloat(Minimum=0, Maximum=1)
       )
      )
      B=1.5
     )
     B=&Lerp_Float(A=1, B=0.5, Alpha=$System.Timeline.NormalizedTime)
    )
    Mesh Scale=&SelectVectorFromArray(
     Array Sampling Mode=::NewEnumerator0
     Direct Array Index=?
     Vector Selection Array=DataInterface() {propertyValues {FloatData [<0.25 1 1>, <0.5 1 1>, <1 1 1>, <1.5 1 1>]}}
     Random Seed=0
    )
    Mesh Scale Mode=::NewEnumerator4
    UsePositionOffset=false
    Write Material Random=true
   ) {module=InitializeParticle001}
   /SetVariables(Particles.VisibilityTag=&MakeIntFromFloat(Float=$User.Variant)) {isSetParameters=true}
   &InitialMeshOrientation(
    Mesh Orientation Mode=::NewEnumerator1
    Rotation=&SelectVectorFromArray(
     Array Sampling Mode=::NewEnumerator0
     Direct Array Index=?
     Vector Selection Array=DataInterface() {propertyValues {FloatData [<0 0 0>, <0 0.25 0>, <0 0 0.25>]}}
     Random Seed=0
    )
    Rotation Coordinate Space=::NewEnumerator0
    Use Rotation Vector=true
   )
   &ShapeLocation(
    Box Size=$User.BoxSize
    MaskSpawnGroup=false
    Override Seed=false
    Shape Origin=&AddVectorToPosition(
     Position=$StackContext.Position
     Vector=&Lerp_Vector(A=<0 0 0>, B=<0 0 -100>, Alpha=$System.Timeline.NormalizedTime)
    )
    Shape Primitive=::NewEnumerator2
    Surface Expansion Mode:ENiagaraExpansionMode=::NewEnumerator1
    Surface Only=true
    Surface Only Band Thickness=40
    UseOwnerScale=true
   )
   &KillParticlesInVolume(
    Box Size:Vector3f=&Subtract_Vector(A=$User.BoxSize, B=<0 0 50>)
    Invert Volume=true
    Kill Shape=::NewEnumerator1
   )
   &AddVelocity(
    Velocity=<0 0 1>
    Velocity Speed Scale=&FloatFromCurve(
     FloatCurve=DataInterface() {propertyValues {Curve {keys [
         {interpMode=RCIM_Cubic, tangentMode=RCTM_User, time=0, value=0}
         {arriveTangent=-0.1536, interpMode=RCIM_Cubic, leaveTangent=-0.1536, tangentMode=RCTM_User, time=0.4, value=1}
         {interpMode=RCIM_Cubic, tangentMode=RCTM_User, time=1, value=1}
        ]}}}
     CurveIndex=&RandomRangeFloat(Minimum=0.75, Maximum=1)
     Scale Curve=&Multiply_Float(A=100, B=&Lerp_Float(A=1, B=3, Alpha=$User.ParaGlideLeanAlpha))
    )
   )
  ]
  particleUpdateScript [
   &ParticleState()
   &SolveForcesAndVelocity(Clamp Velocity=false, Limit Acceleration=false)
   /SetVariables(Particles.distance:NiagaraFloat=&RemapRange(
     Input Value To Remap=&DistanceBetweenPositions(
      A=&TransformPosition(Position=$Particles.Position, Source Space=::Local, Destination Space=::World)
      B=&Add_Vector(A=Linked() {linkedVariable {name=Engine.Owner.Position, type=NiagaraPosition}}, B=<0 0 80>)
     )
     Input Min=80
     Input Max=120
     Output Min=0
     Output Max=1
    )) {isSetParameters=true, module=SetVariables_#8}
   &Color(Color=&LerpLinearColors(StartColor=<rgba 1 1 1 1>, EndColor=<rgba 0 0 0 1>, LerpFactor=$Particles.distance))
   &DynamicMaterialParameters(
    Index 0 Param 1=&FloatFromCurve(FloatCurve=DataInterface() {propertyValues {Curve {keys [
         {interpMode=RCIM_Cubic, tangentMode=RCTM_User, time=0, value=0}
         <at 0.05 1>
         {arriveTangent=-1.0526, interpMode=RCIM_Cubic, leaveTangent=-1.0526, time=0.95, value=1}
         <at 1 0 RCIM_Cubic>
        ]}}})
    Index 0 Param 3=&Multiply_Float(A=&MakeFloatFromVector(Vector=$Particles.Scale, Channel=::NewEnumerator0), B=1)
    Index 0 Param 4=&RandomRangeFloat(Minimum=0, Maximum=1, Randomness Mode=::NewEnumerator1, Random Seed=56)
    Index 1 Param 1=&RandomRangeFloat(Minimum=0, Maximum=1, Randomness Mode=::NewEnumerator1, Random Seed=98)
    Index 1 Param 2=5
    Index 1 Param 3=&FloatFromCurve(FloatCurve=DataInterface() {propertyValues {Curve {keys [{interpMode=RCIM_Cubic, tangentMode=RCTM_User, time=0, value=2}, <at 0.5 1 RCIM_Cubic>, {tangentMode=RCTM_User, time=1, value=2}]}}})
    Index 1 Param 4=&FloatFromCurve(FloatCurve=DataInterface() {propertyValues {Curve {keys [
         {interpMode=RCIM_Cubic, tangentMode=RCTM_User, time=0, value=-1}
         {arriveTangent=3.3333, leaveTangent=3.3333, time=0.4, value=1}
         {arriveTangent=-3.3333, interpMode=RCIM_Cubic, leaveTangent=-3.3333, time=0.6, value=1}
         {tangentMode=RCTM_User, time=1, value=-1}
        ]}}})
    Index 2 Param 1=&FloatFromCurve(
     FloatCurve=DataInterface() {propertyValues {Curve {keys [<at 0 0 RCIM_Constant>,<at 0.2 0.5 RCIM_Constant>,<at 0.4 1 RCIM_Constant>]}}}
     CurveIndex=&RandomRangeFloat(Minimum=0, Maximum=1)
    )
    Index 2 Param 2=&FloatFromCurve(
     FloatCurve=DataInterface() {propertyValues {Curve {keys [<at 0 0 RCIM_Constant>,<at 0.2 0.5 RCIM_Constant>,<at 0.4 1 RCIM_Constant>]}}}
     CurveIndex=&RandomRangeFloat(Minimum=0, Maximum=1)
    )
    Index 3 Param 1=$Particles.distance
    Module.Param0WriteEnabled=true
    Module.Param0WriteEnabled1=true
    Module.Param0WriteEnabled2=true
    Module.Param0WriteEnabled3=true
    Module.Param1WriteEnabled=false
    Module.Param1WriteEnabled1=true
    Module.Param1WriteEnabled2=true
    Module.Param1WriteEnabled3=true
    Module.Param2WriteEnabled=true
    Module.Param2WriteEnabled1=true
    Module.Param2WriteEnabled2=true
    Module.Param2WriteEnabled3=true
    Module.Param3WriteEnabled=true
    Module.Param3WriteEnabled1=true
    Module.Param3WriteEnabled2=true
    Module.Param3WriteEnabled3=true
    Write Parameter Index 1=true
    Write Parameter Index 2=true
    Write Parameter Index 3=true
   )
  ]
  rendererClasses [&NiagaraMeshRendererProperties]
  renderers [
   {rendererClass=&NiagaraMeshRendererProperties, rendererIndex=0}
   {rendererClass=&NiagaraMeshRendererProperties, rendererIndex=1}
   {rendererClass=&NiagaraMeshRendererProperties, rendererIndex=2}
  ]
 }
 DottedTechLines001 {
  simTarget=GPUComputeSim
  emitterUpdateScript [
   &EmitterState()
   &SpawnRate(
    SpawnRate=&Multiply_Float(
     A=&Multiply_Float(
      A=&Lerp_Float(
       A=150
       B=&MakeCustomFloatFromBool(Bool=$User.bIsFrontEndPreview, True Float=250, False Float=650)
       Alpha=$System.Timeline.NormalizedTime
      )
      B=&MakeCustomFloatFromBool(
       Bool=&SetBoolByFloatComparison(A=$User.Variant, B=0, Comparison Type=::NewEnumerator4)
       True Float=1
       False Float=0
      )
     )
     B=1
    )
    Use Spawn Probability=false
   )
   &SpawnRate(SpawnGroup=1, SpawnRate=10, Use Spawn Probability=false) {enabled=false, module=SpawnRate001}
  ]
  particleSpawnScript [
   /SetVariables(Particles.PanSpeed=0.1) {isSetParameters=true, module=SetVariables_#7}
   &InitializeParticle(
    Color Mode=::NewEnumerator0
    Lifetime=&Multiply_Float(
     A=&Add_Float(
      A=&Lerp_Float(
       A=0.5
       B=1
       Alpha=&FloatFromCurve(
        FloatCurve=DataInterface() {propertyValues {Curve {keys [<at 0 0 RCIM_Cubic>, {arriveTangent=1, interpMode=RCIM_Cubic, leaveTangent=1, time=0.7, value=1}, <at 1 1 RCIM_Cubic>]}}}
        CurveIndex=&RandomRangeFloat(Minimum=0, Maximum=1)
       )
      )
      B=1.5
     )
     B=&Lerp_Float(A=1, B=0.5, Alpha=$System.Timeline.NormalizedTime)
    )
    Mesh Scale=&SelectVectorFromArray(
     Array Sampling Mode=::NewEnumerator0
     Direct Array Index=?
     Vector Selection Array=DataInterface() {propertyValues {FloatData [<0.25 1 1>, <0.5 1 1>, <1 1 1>, <1.5 1 1>]}}
     Random Seed=0
    )
    Mesh Scale Mode=::NewEnumerator4
    UsePositionOffset=false
    Write Material Random=true
   ) {module=InitializeParticle001}
   &InitialMeshOrientation(
    Mesh Orientation Mode=::NewEnumerator1
    Rotation=&SelectVectorFromArray(
     Array Sampling Mode=::NewEnumerator0
     Direct Array Index=?
     Vector Selection Array=DataInterface() {propertyValues {FloatData [<0 0 0>, <0 0.25 0>, <0 0 0.25>]}}
     Random Seed=0
    )
    Rotation Coordinate Space=::NewEnumerator0
    Use Rotation Vector=true
   )
   &ShapeLocation(
    Box Size=$User.BoxSize
    MaskSpawnGroup=false
    Override Seed=false
    Shape Origin=&AddVectorToPosition(
     Position=$StackContext.Position
     Vector=&Lerp_Vector(A=<0 0 0>, B=<0 0 -100>, Alpha=$System.Timeline.NormalizedTime)
    )
    Shape Primitive=::NewEnumerator2
    Surface Expansion Mode:ENiagaraExpansionMode=::NewEnumerator1
    Surface Only=true
    Surface Only Band Thickness=40
    UseOwnerScale=true
   )
   &KillParticlesInVolume(
    Box Size:Vector3f=&Subtract_Vector(A=$User.BoxSize, B=<0 0 50>)
    Invert Volume=true
    Kill Shape=::NewEnumerator1
   )
   &AddVelocity(
    Velocity=<0 0 1>
    Velocity Speed Scale=&FloatFromCurve(
     FloatCurve=DataInterface() {propertyValues {Curve {keys [
         {interpMode=RCIM_Cubic, tangentMode=RCTM_User, time=0, value=0}
         {arriveTangent=-0.1536, interpMode=RCIM_Cubic, leaveTangent=-0.1536, tangentMode=RCTM_User, time=0.4, value=1}
         {interpMode=RCIM_Cubic, tangentMode=RCTM_User, time=1, value=1}
        ]}}}
     CurveIndex=&RandomRangeFloat(Minimum=0.75, Maximum=1)
     Scale Curve=&Multiply_Float(A=100, B=&Lerp_Float(A=1, B=3, Alpha=$User.ParaGlideLeanAlpha))
    )
   )
   /SetVariables(Particles.VisibilityTag=&MakeIntFromFloat(Float=$User.Variant)) {isSetParameters=true, module=SetVariables_#9}
  ]
  particleUpdateScript [
   &ParticleState()
   &SolveForcesAndVelocity(Clamp Velocity=false, Limit Acceleration=false)
   /SetVariables(Particles.distance:NiagaraFloat=&RemapRange(
     Input Value To Remap=&DistanceBetweenPositions(
      A=&TransformPosition(Position=$Particles.Position, Source Space=::Local, Destination Space=::World)
      B=&Add_Vector(A=Linked() {linkedVariable {name=Engine.Owner.Position, type=NiagaraPosition}}, B=<0 0 80>)
     )
     Input Min=80
     Input Max=120
     Output Min=0
     Output Max=1
    )) {isSetParameters=true, module=SetVariables_#8}
   &Color(Color=&LerpLinearColors(StartColor=<rgba 1 1 1 1>, EndColor=<rgba 0 0 0 1>, LerpFactor=$Particles.distance))
   &DynamicMaterialParameters(
    Index 0 Param 1=&FloatFromCurve(FloatCurve=DataInterface() {propertyValues {Curve {keys [
         {interpMode=RCIM_Cubic, tangentMode=RCTM_User, time=0, value=0}
         <at 0.05 1>
         {arriveTangent=-1.0526, interpMode=RCIM_Cubic, leaveTangent=-1.0526, time=0.95, value=1}
         <at 1 0 RCIM_Cubic>
        ]}}})
    Index 0 Param 3=&Multiply_Float(A=&MakeFloatFromVector(Vector=$Particles.Scale, Channel=::NewEnumerator0), B=1)
    Index 0 Param 4=&RandomRangeFloat(Minimum=0, Maximum=1, Randomness Mode=::NewEnumerator1, Random Seed=56)
    Index 1 Param 1=&RandomRangeFloat(Minimum=0, Maximum=1, Randomness Mode=::NewEnumerator1, Random Seed=98)
    Index 1 Param 2=5
    Index 1 Param 3=&FloatFromCurve(FloatCurve=DataInterface() {propertyValues {Curve {keys [{interpMode=RCIM_Cubic, tangentMode=RCTM_User, time=0, value=2}, <at 0.5 1 RCIM_Cubic>, {tangentMode=RCTM_User, time=1, value=2}]}}})
    Index 1 Param 4=&FloatFromCurve(FloatCurve=DataInterface() {propertyValues {Curve {keys [
         {interpMode=RCIM_Cubic, tangentMode=RCTM_User, time=0, value=-1}
         {arriveTangent=3.3333, leaveTangent=3.3333, time=0.4, value=1}
         {arriveTangent=-3.3333, interpMode=RCIM_Cubic, leaveTangent=-3.3333, time=0.6, value=1}
         {tangentMode=RCTM_User, time=1, value=-1}
        ]}}})
    Index 2 Param 1=&FloatFromCurve(
     FloatCurve=DataInterface() {propertyValues {Curve {keys [<at 0 0 RCIM_Constant>,<at 0.2 0.5 RCIM_Constant>,<at 0.4 1 RCIM_Constant>]}}}
     CurveIndex=&RandomRangeFloat(Minimum=0, Maximum=1)
    )
    Index 2 Param 2=&FloatFromCurve(
     FloatCurve=DataInterface() {propertyValues {Curve {keys [<at 0 0 RCIM_Constant>,<at 0.2 0.5 RCIM_Constant>,<at 0.4 1 RCIM_Constant>]}}}
     CurveIndex=&RandomRangeFloat(Minimum=0, Maximum=1)
    )
    Index 3 Param 1=$Particles.distance
    Module.Param0WriteEnabled=true
    Module.Param0WriteEnabled1=true
    Module.Param0WriteEnabled2=true
    Module.Param0WriteEnabled3=true
    Module.Param1WriteEnabled=false
    Module.Param1WriteEnabled1=true
    Module.Param1WriteEnabled2=true
    Module.Param1WriteEnabled3=true
    Module.Param2WriteEnabled=true
    Module.Param2WriteEnabled1=true
    Module.Param2WriteEnabled2=true
    Module.Param2WriteEnabled3=true
    Module.Param3WriteEnabled=true
    Module.Param3WriteEnabled1=true
    Module.Param3WriteEnabled2=true
    Module.Param3WriteEnabled3=true
    Write Parameter Index 1=true
    Write Parameter Index 2=true
    Write Parameter Index 3=true
   )
  ]
  rendererClasses [&NiagaraMeshRendererProperties]
  renderers [
   {rendererClass=&NiagaraMeshRendererProperties, rendererIndex=0}
   {rendererClass=&NiagaraMeshRendererProperties, rendererIndex=1}
   {rendererClass=&NiagaraMeshRendererProperties, rendererIndex=2}
  ]
 }
 DottedTechLines002 {
  simTarget=GPUComputeSim
  emitterUpdateScript [
   &EmitterState()
   &SpawnRate(
    SpawnRate=&Multiply_Float(
     A=&Lerp_Float(
      A=150
      B=&MakeCustomFloatFromBool(Bool=$User.bIsFrontEndPreview, True Float=250, False Float=650)
      Alpha=$System.Timeline.NormalizedTime
     )
     B=&MakeCustomFloatFromBool(
      Bool=&SetBoolByFloatComparison(A=$User.Variant, B=1, Comparison Type=::NewEnumerator4)
      True Float=1
      False Float=0
     )
    )
    Use Spawn Probability=false
   )
   &SpawnRate(SpawnGroup=1, SpawnRate=10, Use Spawn Probability=false) {enabled=false, module=SpawnRate001}
  ]
  particleSpawnScript [
   /SetVariables(Particles.PanSpeed=0.1) {isSetParameters=true, module=SetVariables_#7}
   &InitializeParticle(
    Color Mode=::NewEnumerator0
    Lifetime=&Multiply_Float(
     A=&Add_Float(
      A=&Lerp_Float(
       A=0.5
       B=1
       Alpha=&FloatFromCurve(
        FloatCurve=DataInterface() {propertyValues {Curve {keys [<at 0 0 RCIM_Cubic>, {arriveTangent=1, interpMode=RCIM_Cubic, leaveTangent=1, time=0.7, value=1}, <at 1 1 RCIM_Cubic>]}}}
        CurveIndex=&RandomRangeFloat(Minimum=0, Maximum=1)
       )
      )
      B=1.5
     )
     B=&Lerp_Float(A=1, B=0.5, Alpha=$System.Timeline.NormalizedTime)
    )
    Mesh Scale=&SelectVectorFromArray(
     Array Sampling Mode=::NewEnumerator0
     Direct Array Index=?
     Vector Selection Array=DataInterface() {propertyValues {FloatData [<0.25 1 1>, <0.5 1 1>, <1 1 1>, <1.5 1 1>]}}
     Random Seed=0
    )
    Mesh Scale Mode=::NewEnumerator4
    UsePositionOffset=false
    Write Material Random=true
   ) {module=InitializeParticle001}
   &InitialMeshOrientation(
    Mesh Orientation Mode=::NewEnumerator1
    Rotation=&SelectVectorFromArray(
     Array Sampling Mode=::NewEnumerator0
     Direct Array Index=?
     Vector Selection Array=DataInterface() {propertyValues {FloatData [<0 0 0>, <0 0.25 0>, <0 0 0.25>]}}
     Random Seed=0
    )
    Rotation Coordinate Space=::NewEnumerator0
    Use Rotation Vector=true
   )
   &ShapeLocation(
    Box Size=$User.BoxSize
    MaskSpawnGroup=false
    Override Seed=false
    Shape Origin=&AddVectorToPosition(
     Position=$StackContext.Position
     Vector=&Lerp_Vector(A=<0 0 0>, B=<0 0 -100>, Alpha=$System.Timeline.NormalizedTime)
    )
    Shape Primitive=::NewEnumerator2
    Surface Expansion Mode:ENiagaraExpansionMode=::NewEnumerator1
    Surface Only=true
    Surface Only Band Thickness=40
    UseOwnerScale=true
   )
   &KillParticlesInVolume(
    Box Size:Vector3f=&Subtract_Vector(A=$User.BoxSize, B=<0 0 50>)
    Invert Volume=true
    Kill Shape=::NewEnumerator1
   )
   &AddVelocity(
    Velocity=<0 0 1>
    Velocity Speed Scale=&FloatFromCurve(
     FloatCurve=DataInterface() {propertyValues {Curve {keys [
         {interpMode=RCIM_Cubic, tangentMode=RCTM_User, time=0, value=0}
         {arriveTangent=-0.1536, interpMode=RCIM_Cubic, leaveTangent=-0.1536, tangentMode=RCTM_User, time=0.4, value=1}
         {interpMode=RCIM_Cubic, tangentMode=RCTM_User, time=1, value=1}
        ]}}}
     CurveIndex=&RandomRangeFloat(Minimum=0.75, Maximum=1)
     Scale Curve=&Multiply_Float(A=100, B=&Lerp_Float(A=1, B=3, Alpha=$User.ParaGlideLeanAlpha))
    )
   )
   /SetVariables(Particles.VisibilityTag=&MakeIntFromFloat(Float=$User.Variant)) {isSetParameters=true, module=SetVariables_#9}
  ]
  particleUpdateScript [
   &ParticleState()
   &SolveForcesAndVelocity(Clamp Velocity=false, Limit Acceleration=false)
   /SetVariables(Particles.distance:NiagaraFloat=&RemapRange(
     Input Value To Remap=&DistanceBetweenPositions(
      A=&TransformPosition(Position=$Particles.Position, Source Space=::Local, Destination Space=::World)
      B=&Add_Vector(A=Linked() {linkedVariable {name=Engine.Owner.Position, type=NiagaraPosition}}, B=<0 0 80>)
     )
     Input Min=80
     Input Max=120
     Output Min=0
     Output Max=1
    )) {isSetParameters=true, module=SetVariables_#8}
   &Color(Color=&LerpLinearColors(StartColor=<rgba 1 1 1 1>, EndColor=<rgba 0 0 0 1>, LerpFactor=$Particles.distance))
   &DynamicMaterialParameters(
    Index 0 Param 1=&FloatFromCurve(FloatCurve=DataInterface() {propertyValues {Curve {keys [
         {interpMode=RCIM_Cubic, tangentMode=RCTM_User, time=0, value=0}
         <at 0.05 1>
         {arriveTangent=-1.0526, interpMode=RCIM_Cubic, leaveTangent=-1.0526, time=0.95, value=1}
         <at 1 0 RCIM_Cubic>
        ]}}})
    Index 0 Param 3=&Multiply_Float(A=&MakeFloatFromVector(Vector=$Particles.Scale, Channel=::NewEnumerator0), B=1)
    Index 0 Param 4=&RandomRangeFloat(Minimum=0, Maximum=1, Randomness Mode=::NewEnumerator1, Random Seed=56)
    Index 1 Param 1=&RandomRangeFloat(Minimum=0, Maximum=1, Randomness Mode=::NewEnumerator1, Random Seed=98)
    Index 1 Param 2=5
    Index 1 Param 3=&FloatFromCurve(FloatCurve=DataInterface() {propertyValues {Curve {keys [{interpMode=RCIM_Cubic, tangentMode=RCTM_User, time=0, value=2}, <at 0.5 1 RCIM_Cubic>, {tangentMode=RCTM_User, time=1, value=2}]}}})
    Index 1 Param 4=&FloatFromCurve(FloatCurve=DataInterface() {propertyValues {Curve {keys [
         {interpMode=RCIM_Cubic, tangentMode=RCTM_User, time=0, value=-1}
         {arriveTangent=3.3333, leaveTangent=3.3333, time=0.4, value=1}
         {arriveTangent=-3.3333, interpMode=RCIM_Cubic, leaveTangent=-3.3333, time=0.6, value=1}
         {tangentMode=RCTM_User, time=1, value=-1}
        ]}}})
    Index 2 Param 1=&FloatFromCurve(
     FloatCurve=DataInterface() {propertyValues {Curve {keys [<at 0 0 RCIM_Constant>,<at 0.2 0.5 RCIM_Constant>,<at 0.4 1 RCIM_Constant>]}}}
     CurveIndex=&RandomRangeFloat(Minimum=0, Maximum=1)
    )
    Index 2 Param 2=&FloatFromCurve(
     FloatCurve=DataInterface() {propertyValues {Curve {keys [<at 0 0 RCIM_Constant>,<at 0.2 0.5 RCIM_Constant>,<at 0.4 1 RCIM_Constant>]}}}
     CurveIndex=&RandomRangeFloat(Minimum=0, Maximum=1)
    )
    Index 3 Param 1=$Particles.distance
    Module.Param0WriteEnabled=true
    Module.Param0WriteEnabled1=true
    Module.Param0WriteEnabled2=true
    Module.Param0WriteEnabled3=true
    Module.Param1WriteEnabled=false
    Module.Param1WriteEnabled1=true
    Module.Param1WriteEnabled2=true
    Module.Param1WriteEnabled3=true
    Module.Param2WriteEnabled=true
    Module.Param2WriteEnabled1=true
    Module.Param2WriteEnabled2=true
    Module.Param2WriteEnabled3=true
    Module.Param3WriteEnabled=true
    Module.Param3WriteEnabled1=true
    Module.Param3WriteEnabled2=true
    Module.Param3WriteEnabled3=true
    Write Parameter Index 1=true
    Write Parameter Index 2=true
    Write Parameter Index 3=true
   )
  ]
  rendererClasses [&NiagaraMeshRendererProperties]
  renderers [
   {rendererClass=&NiagaraMeshRendererProperties, rendererIndex=0}
   {rendererClass=&NiagaraMeshRendererProperties, rendererIndex=1}
   {rendererClass=&NiagaraMeshRendererProperties, rendererIndex=2}
  ]
 }
 DottedTechLines003 {
  simTarget=GPUComputeSim
  emitterUpdateScript [
   &EmitterState()
   &SpawnRate(
    SpawnRate=&Multiply_Float(
     A=&Lerp_Float(
      A=150
      B=&MakeCustomFloatFromBool(Bool=$User.bIsFrontEndPreview, True Float=250, False Float=650)
      Alpha=$System.Timeline.NormalizedTime
     )
     B=&MakeCustomFloatFromBool(
      Bool=&SetBoolByFloatComparison(A=$User.Variant, B=2, Comparison Type=::NewEnumerator4)
      True Float=1
      False Float=0
     )
    )
    Use Spawn Probability=false
   )
   &SpawnRate(SpawnGroup=1, SpawnRate=10, Use Spawn Probability=false) {enabled=false, module=SpawnRate001}
  ]
  particleSpawnScript [
   /SetVariables(Particles.PanSpeed=0.1) {isSetParameters=true, module=SetVariables_#7}
   &InitializeParticle(
    Color Mode=::NewEnumerator0
    Lifetime=&Multiply_Float(
     A=&Add_Float(
      A=&Lerp_Float(
       A=0.5
       B=1
       Alpha=&FloatFromCurve(
        FloatCurve=DataInterface() {propertyValues {Curve {keys [<at 0 0 RCIM_Cubic>, {arriveTangent=1, interpMode=RCIM_Cubic, leaveTangent=1, time=0.7, value=1}, <at 1 1 RCIM_Cubic>]}}}
        CurveIndex=&RandomRangeFloat(Minimum=0, Maximum=1)
       )
      )
      B=1.5
     )
     B=&Lerp_Float(A=1, B=0.5, Alpha=$System.Timeline.NormalizedTime)
    )
    Mesh Scale=&SelectVectorFromArray(
     Array Sampling Mode=::NewEnumerator0
     Direct Array Index=?
     Vector Selection Array=DataInterface() {propertyValues {FloatData [<0.25 1 1>, <0.5 1 1>, <1 1 1>, <1.5 1 1>]}}
     Random Seed=0
    )
    Mesh Scale Mode=::NewEnumerator4
    UsePositionOffset=false
    Write Material Random=true
   ) {module=InitializeParticle001}
   &InitialMeshOrientation(
    Mesh Orientation Mode=::NewEnumerator1
    Rotation=&SelectVectorFromArray(
     Array Sampling Mode=::NewEnumerator0
     Direct Array Index=?
     Vector Selection Array=DataInterface() {propertyValues {FloatData [<0 0 0>, <0 0.25 0>, <0 0 0.25>]}}
     Random Seed=0
    )
    Rotation Coordinate Space=::NewEnumerator0
    Use Rotation Vector=true
   )
   &ShapeLocation(
    Box Size=$User.BoxSize
    MaskSpawnGroup=false
    Override Seed=false
    Shape Origin=&AddVectorToPosition(
     Position=$StackContext.Position
     Vector=&Lerp_Vector(A=<0 0 0>, B=<0 0 -100>, Alpha=$System.Timeline.NormalizedTime)
    )
    Shape Primitive=::NewEnumerator2
    Surface Expansion Mode:ENiagaraExpansionMode=::NewEnumerator1
    Surface Only=true
    Surface Only Band Thickness=40
    UseOwnerScale=true
   )
   &KillParticlesInVolume(
    Box Size:Vector3f=&Subtract_Vector(A=$User.BoxSize, B=<0 0 50>)
    Invert Volume=true
    Kill Shape=::NewEnumerator1
   )
   &AddVelocity(
    Velocity=<0 0 1>
    Velocity Speed Scale=&FloatFromCurve(
     FloatCurve=DataInterface() {propertyValues {Curve {keys [
         {interpMode=RCIM_Cubic, tangentMode=RCTM_User, time=0, value=0}
         {arriveTangent=-0.1536, interpMode=RCIM_Cubic, leaveTangent=-0.1536, tangentMode=RCTM_User, time=0.4, value=1}
         {interpMode=RCIM_Cubic, tangentMode=RCTM_User, time=1, value=1}
        ]}}}
     CurveIndex=&RandomRangeFloat(Minimum=0.75, Maximum=1)
     Scale Curve=&Multiply_Float(A=100, B=&Lerp_Float(A=1, B=3, Alpha=$User.ParaGlideLeanAlpha))
    )
   )
   /SetVariables(Particles.VisibilityTag=&MakeIntFromFloat(Float=$User.Variant)) {isSetParameters=true, module=SetVariables_#9}
  ]
  particleUpdateScript [
   &ParticleState()
   &SolveForcesAndVelocity(Clamp Velocity=false, Limit Acceleration=false)
   /SetVariables(Particles.distance:NiagaraFloat=&RemapRange(
     Input Value To Remap=&DistanceBetweenPositions(
      A=&TransformPosition(Position=$Particles.Position, Source Space=::Local, Destination Space=::World)
      B=&Add_Vector(A=Linked() {linkedVariable {name=Engine.Owner.Position, type=NiagaraPosition}}, B=<0 0 80>)
     )
     Input Min=80
     Input Max=120
     Output Min=0
     Output Max=1
    )) {isSetParameters=true, module=SetVariables_#8}
   &Color(Color=&LerpLinearColors(StartColor=<rgba 1 1 1 1>, EndColor=<rgba 0 0 0 1>, LerpFactor=$Particles.distance))
   &DynamicMaterialParameters(
    Index 0 Param 1=&FloatFromCurve(FloatCurve=DataInterface() {propertyValues {Curve {keys [
         {interpMode=RCIM_Cubic, tangentMode=RCTM_User, time=0, value=0}
         <at 0.05 1>
         {arriveTangent=-1.0526, interpMode=RCIM_Cubic, leaveTangent=-1.0526, time=0.95, value=1}
         <at 1 0 RCIM_Cubic>
        ]}}})
    Index 0 Param 3=&Multiply_Float(A=&MakeFloatFromVector(Vector=$Particles.Scale, Channel=::NewEnumerator0), B=1)
    Index 0 Param 4=&RandomRangeFloat(Minimum=0, Maximum=1, Randomness Mode=::NewEnumerator1, Random Seed=56)
    Index 1 Param 1=&RandomRangeFloat(Minimum=0, Maximum=1, Randomness Mode=::NewEnumerator1, Random Seed=98)
    Index 1 Param 2=5
    Index 1 Param 3=&FloatFromCurve(FloatCurve=DataInterface() {propertyValues {Curve {keys [{interpMode=RCIM_Cubic, tangentMode=RCTM_User, time=0, value=2}, <at 0.5 1 RCIM_Cubic>, {tangentMode=RCTM_User, time=1, value=2}]}}})
    Index 1 Param 4=&FloatFromCurve(FloatCurve=DataInterface() {propertyValues {Curve {keys [
         {interpMode=RCIM_Cubic, tangentMode=RCTM_User, time=0, value=-1}
         {arriveTangent=3.3333, leaveTangent=3.3333, time=0.4, value=1}
         {arriveTangent=-3.3333, interpMode=RCIM_Cubic, leaveTangent=-3.3333, time=0.6, value=1}
         {tangentMode=RCTM_User, time=1, value=-1}
        ]}}})
    Index 2 Param 1=&FloatFromCurve(
     FloatCurve=DataInterface() {propertyValues {Curve {keys [<at 0 0 RCIM_Constant>,<at 0.2 0.5 RCIM_Constant>,<at 0.4 1 RCIM_Constant>]}}}
     CurveIndex=&RandomRangeFloat(Minimum=0, Maximum=1)
    )
    Index 2 Param 2=&FloatFromCurve(
     FloatCurve=DataInterface() {propertyValues {Curve {keys [<at 0 0 RCIM_Constant>,<at 0.2 0.5 RCIM_Constant>,<at 0.4 1 RCIM_Constant>]}}}
     CurveIndex=&RandomRangeFloat(Minimum=0, Maximum=1)
    )
    Index 3 Param 1=$Particles.distance
    Module.Param0WriteEnabled=true
    Module.Param0WriteEnabled1=true
    Module.Param0WriteEnabled2=true
    Module.Param0WriteEnabled3=true
    Module.Param1WriteEnabled=false
    Module.Param1WriteEnabled1=true
    Module.Param1WriteEnabled2=true
    Module.Param1WriteEnabled3=true
    Module.Param2WriteEnabled=true
    Module.Param2WriteEnabled1=true
    Module.Param2WriteEnabled2=true
    Module.Param2WriteEnabled3=true
    Module.Param3WriteEnabled=true
    Module.Param3WriteEnabled1=true
    Module.Param3WriteEnabled2=true
    Module.Param3WriteEnabled3=true
    Write Parameter Index 1=true
    Write Parameter Index 2=true
    Write Parameter Index 3=true
   )
  ]
  rendererClasses [&NiagaraMeshRendererProperties]
  renderers [
   {rendererClass=&NiagaraMeshRendererProperties, rendererIndex=0}
   {rendererClass=&NiagaraMeshRendererProperties, rendererIndex=1}
   {rendererClass=&NiagaraMeshRendererProperties, rendererIndex=2}
  ]
 }
 LongBlueStripe {
  emitterUpdateScript [
   &EmitterState(Life Cycle Mode=::NewEnumerator1, Loop Behavior=::NewEnumerator1, Loop Duration Mode=::NewEnumerator1, UseLoopDelay=false)
   &SpawnBurst_Instantaneous(Spawn Count=2, Use Spawn Probability=false, UseLoopCountLimit=false)
  ]
  particleSpawnScript [
   &PartitionParticles()
   &InitializeParticle(
    Color Mode=::NewEnumerator0
    Lifetime=5
    Mesh Scale Mode=::NewEnumerator3
    Ribbon Width=35
    Ribbon Width Mode=::NewEnumerator1
    Sprite Rotation Angle=90
    Sprite Rotation Mode=::NewEnumerator0
    Sprite Size=<300 7>
    Sprite Size Mode=::NewEnumerator4
    UsePositionOffset=false
    Write Material Random=true
   )
   &SpriteFacingAndAlignment(Sprite Alignment=<0 0 1>, WriteAlignment=true, WriteFacing=false)
   &Color()
   &InitialMeshOrientation(
    Mesh Orientation Mode=::NewEnumerator2
    Rotation=&SelectVectorFromArray(
     Direct Array Index=$Particles.PartitionParticles.Partition
     Vector Selection Array=DataInterface() {propertyValues {FloatData [<0 0.25 0>, <0 0.25 0>]}}
    )
    Use Rotation Vector=true
   )
   /SetVariables(Particles.VisibilityTag=&MakeIntFromFloat(Float=$User.Variant)) {isSetParameters=true}
  ]
  particleUpdateScript [
   &ParticleState(Kill Particles When Lifetime Has Elapsed=false, Let Infinitely Lived Particles Die When Emitter Deactivates=true)
   &SkeletalMeshLocation(
    Bone / Socket Index=$Particles.PartitionParticles.Partition
    MaskSpawnGroup=false
    Mesh Sampling Type=::NewEnumerator5
    Orientation Sampling=::NewEnumerator1
    Skeletal Mesh=$User.SkeletalMesh
    Socket Sampling Mode=::NewEnumerator5
   )
   &DynamicMaterialParameters(
    Index 0 Param 1=&RandomRangeFloat(Minimum=0, Maximum=1)
    Index 0 Param 2=&RandomRangeFloat(Minimum=0, Maximum=1)
    Index 0 Param 3=&Lerp_Float(A=1, B=0.15, Alpha=$System.Timeline.NormalizedTime)
    Index 0 Param 4=&Lerp_Float(A=1, B=0.15, Alpha=$System.Timeline.NormalizedTime)
    Module.Param0WriteEnabled=true
    Module.Param1WriteEnabled=true
    Module.Param2WriteEnabled=true
    Module.Param3WriteEnabled=true
   )
  ]
  rendererClasses [&NiagaraSpriteRendererProperties]
  renderers [
   {rendererClass=&NiagaraSpriteRendererProperties}
   {rendererClass=&NiagaraSpriteRendererProperties, rendererIndex=1}
   {rendererClass=&NiagaraSpriteRendererProperties, rendererIndex=2}
  ]
 }
}
summary {
 systemName=NS_Contrail_IndieBucket
 emitters [
  {bEnabled=false, emitterName=DottedTechLines, simTarget=GPUComputeSim, rendererClasses [&NiagaraMeshRendererProperties]}
  {emitterName=LongBlueStripe, rendererClasses [&NiagaraSpriteRendererProperties]}
  {emitterName=DottedTechLines001, simTarget=GPUComputeSim, rendererClasses [&NiagaraMeshRendererProperties]}
  {emitterName=DottedTechLines002, simTarget=GPUComputeSim, rendererClasses [&NiagaraMeshRendererProperties]}
  {emitterName=DottedTechLines003, simTarget=GPUComputeSim, rendererClasses [&NiagaraMeshRendererProperties]}
 ]
 userVariables [
  User.bIsFrontEnd {type=NiagaraBool}
  User.bIsFrontEndPreview {type=NiagaraBool}
  User.BoxSize {type=Vector3f, description="", defaultValue {struct=Vector3f, value=<150 200 400>}}
  User.ParaGlideLeanAlpha {type=NiagaraFloat, defaultValue=0, description=""}
  User.RibbonIDOffset {type=NiagaraInt32}
  User.SkeletalMesh {
   type=NiagaraDataInterfaceSkeletalMesh
   defaultValue {
    struct=NiagaraExt_VariableValue_DataInterface
    value {dataInterface=&self:NiagaraDataInterfaceSkeletalMesh_2}
   }
  }
  User.SkelMesh {
   type=NiagaraDataInterfaceSkeletalMesh
   defaultValue {
    struct=NiagaraExt_VariableValue_DataInterface
    value {dataInterface=&self:NiagaraDataInterfaceSkeletalMesh_0}
   }
  }
  User.Variant {type=NiagaraFloat, defaultValue=0, description=""}
  User.VariantID {type=NiagaraFloat, defaultValue=0, description=""}
 ]
}
